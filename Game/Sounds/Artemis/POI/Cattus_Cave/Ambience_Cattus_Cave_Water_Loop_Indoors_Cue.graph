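# Sound cue for the looping water ambience of the Cattus Cave POI, routed to the indoor ambient sound class.
# A looping wave player feeds an Enveloper whose volume curve ramps up from silence at the start, with overall volume reduced.
# Attenuation override is enabled, so the cue's own non-spatialized distance falloff replaces the referenced lake-water attenuation asset.
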SoundCue /Game/Sounds/Artemis/POI/Cattus_Cave/Ambience_Cattus_Cave_Water_Loop_Indoors_Cue
root=Enveloper_0
nodes [
 {
  id=Enveloper_0
  children [WavePlayer_2]
  properties {pitchCurve {editorCurveData {keys [{time=0}]}}, volumeCurve {editorCurveData {keys [<at 0 0>,<at 1 1>]}}}
 }
 {id=WavePlayer_2, properties {bLooping=true, soundWaveAssetPtr=&/Ambience_Cattus_Cave_Water_Loop}}
]
properties {
 attenuationSettings=/Game/Athena/Sounds/Ambient/Fort_Amb_Water_Lake_Att
 bOverrideAttenuation=true
 maxDistance=2425
 soundClassObject=&Fort_Ambient_Indoors_Athena
 volumeMultiplier=0.501
 attenuationOverrides {bSpatialize=false, falloffDistance=1500, nonSpatializedRadiusStart=700, stereoSpread=1000, attenuationShapeExtents {x=925}}
 concurrencyOverrides {maxCount=1, resolutionRule=StopOldest}
}
